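# Material function for a weapon-skin temperature gauge, blended into the input material's emissive by Alpha.
# A scaled, offsettable UV gauge with outline sits over a parallax hex-grid background, with a panning texture modulating fill brightness.
# Fill level comes from a world-XY gradient texture scaled by time of day, colored between min and max temperature colors along the gauge mask.
# The Temperature parameter and gauge bump-offset branch are unconnected.
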MaterialFunction /BRCosmetics/Weapons/WeaponSkins/MaterialFunction/TemperatureMeter
graph {nodes [
  n1 = FunctionInput() {inputName=Material, inputType=FunctionInput_MaterialAttributes}
  n2 = GetMaterialAttributes(None=@n1) {attributeGetTypes [#1]}
  n3 = Constant3Vector() {constant=<rgba 0.005 0.0132 0.02 1>}
  n4 = Constant3Vector() {constant=<rgba 0.0261 0.0985 0.1354 1>}
  n5 = TextureSample(UVs=BumpOffset(Coordinate=TextureCoordinate() {uTiling=40, vTiling=40}, Height=ScalarParameter() {defaultValue=-0.5, parameterName=BG Height})) {samplerType=SAMPLERTYPE_Masks, texture=/Game/Accessories/FORT_Backpacks/F_MED_CyberRunner_Backpack/Textures/HexGrid_M}
  n6 = ScalarParameter() {defaultValue=2, parameterName=Guage Brightness}
  n7 = Add(
   A=LinearInterpolate(
    A=TextureCoordinate()
    B=TextureCoordinate() {coordinateIndex=1}
    Alpha=ScalarParameter() {group=Temperature Guage, parameterName=use Second UV Channel}
   )
   B=AppendVector(
    A=ScalarParameter() {defaultValue=-0.19, group=Temperature Guage, parameterName=Guage X}
    B=ScalarParameter() {defaultValue=-0.122, group=Temperature Guage, parameterName=Guage Y}
   )
  )
  n8 = MaterialFunctionCall(UVs=@n7, Texture Scale=ScalarParameter() {defaultValue=0.135, group=Temperature Guage, parameterName=TemperatureGuage_Scale}) {materialFunction=&ScaleUVsByCenter}
  n9 = Multiply(
   A=TextureSample(UVs=@n8.UVs) {texture=/BRCosmetics/Accessories/FORT_Backpacks/Backpack_M_MED_Mr_Ice_Guy/Textures/TemperatureGuage_E}.RGB
   B=@n8.0-1 mask
  )
  n10 = VectorParameter() {parameterName=Min Temp Color, defaultValue=<rgba 0 0.3958 1 1>}
  n11 = VectorParameter() {parameterName=Max Temp Color, defaultValue=<rgba 1 0 0 1>}
  n12 = TextureSample(UVs=@n8.UVs) {
   samplerSource=SSM_Clamp_WorldGroupSettings
   samplerType=SAMPLERTYPE_Masks
   texture=/BRCosmetics/Accessories/FORT_Backpacks/Backpack_M_MED_Mr_Ice_Guy/Textures/TemperatureGuage_M
  }
  n13 = MaterialFunctionCall(Float3=Multiply(A=@n12.RGB, B=@n8.0-1 mask)) {materialFunction=&BreakOutFloat3Components}
  n14 = Saturate(None=Power(Base=Multiply(A=@n13.G) {constB=1.2}) {constExponent=1.5})
  n15 = OneMinus(None=@n13.G)
  n16 = Divide(A=ComponentMask(None=WorldPosition() {worldPositionShaderOffset=WPT_ExcludeAllShaderOffsets}.XYZ) {g=true, r=true}, B=Constant() {r=5599})
  n17 = TextureSample(UVs=@n16) {texture=/BRCosmetics/Packages/DS_Fortnite_Walls/T_FORT_Wall_Gradient}
  n18 = Constant() {r=0.5}
  n19 = MaterialFunctionCall(Value_Morning=@n18, Value_Day=Constant() {r=1}, Value_Evening=@n18, Value_Night=Constant() {r=0.1}) {materialFunction=&Scalar_Time_of_Day}
  n20 = LinearInterpolate(Alpha=Saturate(None=Multiply(A=@n17.R, B=@n19.Result))) {constA=0.14, constB=0.99}
  n21 = Multiply(
   A=Multiply(A=LinearInterpolate(A=@n10.RGB, B=@n11.RGB, Alpha=@n14), B=@n13.B)
   B=Add(A=Subtract(A=@n15), B=Multiply(A=@n20, B=@n20)) {constA=-1}
  )
  n22 = Multiply(A=@n21) {constB=0.66}
  n23 = TextureSample(UVs=Panner(Coordinate=TextureCoordinate() {uTiling=16, vTiling=16}) {speedX=2}) {texture=/BRCosmetics/Weapons/WeaponSkins/Wraps/Textures/ComicWarp/T_Wrap_ComicWarp_Grad}
  n24 = Add(
   A=@n9
   B=LinearInterpolate(
    A=LinearInterpolate(A=@n21, B=@n22, Alpha=@n23.R)
    B=VectorParameter() {parameterName=Outline Color, defaultValue=<rgba 0.5397 0.7411 0.901 1>}.RGB
    Alpha=@n13.R
   )
  )
  n25 = ComponentMask(None=Saturate(None=Multiply(A=@n9) {constB=500})) {r=true}
  n26 = Max(A=@n13.R, B=@n13.R)
  n27 = Add(
   A=@n2.EmissiveColor
   B=LinearInterpolate(A=LinearInterpolate(A=@n3, B=@n4, Alpha=@n5.R), B=Multiply(A=@n6, B=@n24), Alpha=Max(A=@n25, B=@n26))
  )
  FunctionOutput(None=BlendMaterialAttributes(
    A=@n1
    B=SetMaterialAttributes!(MaterialAttributes=@n2.MaterialAttributes, EmissiveColor=@n27) {attributeSetTypes [#1]}
    Alpha=FunctionInput() {inputName=Alpha, inputType=FunctionInput_Scalar}
   ))
  LinearInterpolate(
   A=@n7
   B=BumpOffset(Coordinate=@n7, Height=ScalarParameter() {group=Temperature Guage, parameterName=Guage Bump Height})
   Alpha=ScalarParameter() {group=Temperature Guage, parameterName=Guage BumpOffset Intensity}
  )
  LinearInterpolate()
  OneMinus(None=MaterialFunctionCall(UV Channel=@n16) {materialFunction=&LinearGradient}.UGradient)
  ScalarParameter() {defaultValue=0.5, parameterName=Temperature}
  TextureCoordinate()
 ]}
