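# Material function for the Cyborg Universe weapon skin's "Negative" effect, adding an animated HDR glow, cyan-blue by default, to incoming emissive.
# Panned Perlin noise, with sine-pulsed intensity, warps the UVs of a second Perlin sample that modulates the color.
# Normal-based Fresnel controls noise blending and fades the glow toward edges, and the Gmap input masks it.
# The Alpha input blends the result with the original material attributes.
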MaterialFunction /BRCosmetics/Weapons/WeaponSkins/MaterialFunction/MF_CyborgUniverseNegative
graph {nodes [
  n1 = Multiply(A=TextureCoordinate(), B=ScalarParameter() {defaultValue=3.5, group=Negative, parameterName=NoiseTiling})
  n2 = ScalarParameter() {defaultValue=0.05, group=Negative, parameterName=WaterPanner}
  n3 = Constant()
  n4 = TextureSample(UVs=Panner(Coordinate=@n1, Speed=AppendVector(A=@n2, B=@n3))) {texture=/Game/Global/Textures/Default/Noise/3xPerlin_M}
  n5 = ScalarParameter() {defaultValue=0.25, group=Negative, parameterName=WaterWarpIntensity}
  n6 = Sine(None=Time()) {period=10}
  n7 = Multiply(A=AppendVector(A=@n4.R, B=@n4.R), B=LinearInterpolate(A=@n5, B=Multiply(A=@n5) {constB=2}, Alpha=ConstantBiasScale(None=@n6) {bias=0.5}))
  n8 = FunctionInput() {inputType=FunctionInput_MaterialAttributes}
  n9 = GetMaterialAttributes(None=@n8) {attributeGetTypes [#1, #2]}
  n10 = Transform(None=@n9.Normal)
  Add(A=@n7, B=Multiply(A=ComponentMask(None=@n10) {r=true}) {constB=0.4})
  n11 = VectorParameter() {group=Negative, parameterName=Universe_NegativeColor, defaultValue=<rgba 0 232.3965 450 1>}
  n12 = Multiply(A=@n1) {constB=0.5}
  n13 = Power(Base=TextureSample(UVs=Add(A=@n12, B=@n7)) {texture=&3xPerlin_M}.R)
  n14 = Fresnel(ExponentIn=ScalarParameter() {defaultValue=0.02, group=Negative, parameterName=Fres_EX_Negative}, Normal=@n10) {baseReflectFraction=0}
  n15 = Multiply(A=@n14) {constB=0.9}
  n16 = ScalarParameter() {defaultValue=0.4048, group=Negative, parameterName=WaterWarpPower}
  n17 = LinearInterpolate(A=@n11.RGB, B=Multiply(A=@n11.RGB, B=Saturate(None=@n13)), Alpha=Saturate(None=Power(Base=@n15, Exp=@n16) {constExponent=4}))
  n18 = Add(A=@n9.EmissiveColor, B=LinearInterpolate(B=LinearInterpolate(A=@n17, Alpha=@n14) {constB=0}, Alpha=FunctionInput() {inputName=Gmap}))
  FunctionOutput(None=BlendMaterialAttributes(
    A=@n8
    B=SetMaterialAttributes!(MaterialAttributes=@n9.MaterialAttributes, EmissiveColor=@n18) {attributeSetTypes [#1]}
    Alpha=FunctionInput() {inputName=Alpha, inputType=FunctionInput_Scalar}
   ))
 ]}
